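# Niagara swing trail for the DiscoButler pickaxe, combining blue and water ribbons with tumbling ice cubes.
# TrailBlue and TrailWater are short-lived anim-trail ribbons tinted by time-of-day color, with TrailBlue's bone indices offset by a right-swing value.
# IceCubes spawns mesh particles around a socket that inherit some velocity, fall under gravity and shrink away.
# Shared system fade values drive ribbon alpha and ice spawn probability.
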NiagaraSystem /BRCosmetics/Weapons/FORT_Melee/Pickaxe_DiscoButler/FX/NS_Pickaxe_DiscoButler_Trail
dependencies {
 usedDataInterfaces [
  &NiagaraDataInterfaceSkeletalMesh
  &NiagaraDataInterfaceCurve
  &NiagaraDataInterfaceSocketReader
  &NiagaraDataInterfaceArrayInt32
 ]
 usedDynamicInputs [
  &Multiply_Vector2D_ByFloat
  &UniformAOrBFloat
  &RandomRangeVector2D
  &MergeHelper_FloatSwitch
  &FNTimeOfDayColor
  &ClampFloat
  &RemapRange
  &Add_Float
  &Subtract_Float
  &SelectIntFromArray
 ]
 usedModules [
  &self:SystemSpawnScript.SystemScriptSource.SystemScriptGraph.NiagaraNodeAssignment_0.SetVariables_#1
  &self:SystemSpawnScript.SystemScriptSource.SystemScriptGraph.NiagaraNodeAssignment_10.SetVariables_#2
  &SystemState
  &self:SystemSpawnScript.SystemScriptSource.SystemScriptGraph.NiagaraNodeAssignment_6.SetVariables_#3
  &self:SystemSpawnScript.SystemScriptSource.SystemScriptGraph.NiagaraNodeAssignment_5.SetVariables_#4
  &self:SystemSpawnScript.SystemScriptSource.SystemScriptGraph.NiagaraNodeAssignment_8.SetVariables_#5
 ]
}
emitters {
 IceCubes {
  emitterUpdateScript [
   &EmitterState()
   &SpawnRate(
    Spawn Probability=&Multiply_Float(A=$System.FadeSpawn, B=$System.FadeSystem)
    SpawnRate=30
    Use Spawn Probability=true
   ) {module=SpawnRate001}
  ]
  particleSpawnScript [
   &InitializeParticle(
    Color=<rgba 1.5 1.5 1.5 0.85>
    Lifetime Max=0.35
    Lifetime Min=0.25
    Lifetime Mode=::NewEnumerator1
    Mesh Scale Max=<0.28 0.28 0.28>
    Mesh Scale Min=<0.16 0.16 0.16>
    Mesh Scale Mode=::NewEnumerator2
    Sprite Rotation Mode=::NewEnumerator2
    UsePositionOffset=false
    Write Material Random=true
   ) {module=InitializeParticle001}
   &SocketLocation(
    SamplePosOffset=true
    Sampled Position Offset=&RandomRangeVector(Minimum=<0 -9 -19>, Maximum=<0 9 19>)
    Socket Index=4
    Socket Reader=$System.Socket Reader
    Socket Sampling Mode=::NewEnumerator3
    VelocitySpeedLimit=false
   )
   &InheritVelocity(
    Inherited Velocity Amount Scale=&Multiply_VectorByFloat(Vector=<1 1 1>, Float=&RandomRangeFloat(Minimum=0.03, Maximum=0.12))
    Inherited Velocity Speed Limit=200
    Speed Limit=true
    UseSpeedThreshold=false
   ) {module=InheritVelocity001}
   &AddVelocity(
    Velocity=&RandomVector(Vector Scale=1)
    Velocity Speed Scale=&RandomRangeFloat(Minimum=5, Maximum=30)
   )
   &ApplyInitialForces(Apply Rotational Force to Rotational Velocity=false)
   &InitialMeshOrientation(Use Rotation Vector=false)
  ]
  particleUpdateScript [
   &ParticleState()
   &UpdateMeshOrientation(
    Rotation Rate=&RandomRangeFloat(Minimum=0, Maximum=3)
    Rotation Vector=&RandomVector(Vector Scale=1)
   )
   &ScaleMeshSize(Scale Factor=&VectorFromFloat(Value=&FloatFromCurve(FloatCurve=DataInterface() {propertyValues {Curve {keys [
          {arriveTangent=3.8061, interpMode=RCIM_Cubic, leaveTangent=3.8061, tangentMode=RCTM_User, time=0, value=0.5}
          {arriveTangent=0.6588, interpMode=RCIM_Cubic, leaveTangent=0.6588, time=0.2, value=1}
          {arriveTangent=-1.25, interpMode=RCIM_Cubic, leaveTangent=-1.25, time=0.7589, value=1}
          {arriveTangent=-2.2098, interpMode=RCIM_Cubic, leaveTangent=-2.2098, tangentMode=RCTM_User, time=1, value=0}
         ]}}})))
   &GravityForce(Gravity=&RandomRangeVector(Minimum=<0 0 -400>, Maximum=<0 0 -550>)) {module=GravityForce001}
   &SolveForcesAndVelocity(Clamp Velocity=false, Limit Acceleration=false)
  ]
  rendererClasses [&NiagaraMeshRendererProperties]
  renderers [{rendererClass=&NiagaraMeshRendererProperties, rendererIndex=0}]
 }
 TrailBlue {
  emitterUpdateScript [
   &EmitterState(Life Cycle Mode=::NewEnumerator1, Loop Behavior=::NewEnumerator1, Loop Duration=0.27, UseLoopDelay=true)
   &FN_SpawnAnimTrail()
  ]
  particleSpawnScript [
   &FN_InitializeAnimTrailBoneIndecies(
    First Bone or Socket Index:NiagaraInt32=&Add_Integer(A=0, B=$System.RightSwing)
    Second Bone or Socket Index:NiagaraInt32=&Add_Integer(A=1, B=$System.RightSwing)
   ) {module=FN_InitializeAnimTrailBoneIndecies001}
   &FN_InitializeAnimTrail(
    Color=&MultiplyLinearColor(
     Linear Color=<rgba 0.04 0.19 0.4 0.63>
     Scale Factor=&MakeVector4FromVectorAndFloat("Vector (RGB)"=<1 1 1>, "Float (Alpha)"=&Multiply_Float(A=$System.FadeSpawn, B=$System.FadeSpawnOut))
    )
    Lifetime=0.13
   ) {module=FN_InitializeAnimTrail001}
  ]
  particleUpdateScript [
   &FN_UpdateAnimTrail()
   &ScaleColor(
    Scale Alpha=&FloatFromCurve(FloatCurve=DataInterface() {propertyValues {Curve {keys [<at 0 0>,<at 0.0468 0>,<at 0.3369 1>,<at 1 1>]}}}, Scale Curve=$System.FadeSystem)
    ScaleA=true
    ScaleRGB=true
   )
   &ScaleColor(Color Scale:LinearColor=$System.TODColor, Scale Mode=3) {module=ScaleColor001}
  ]
  rendererClasses [&NiagaraRibbonRendererProperties]
  renderers [{rendererClass=&NiagaraRibbonRendererProperties}]
 }
 TrailWater {
  emitterUpdateScript [
   &EmitterState(Life Cycle Mode=::NewEnumerator1, Loop Behavior=::NewEnumerator1, Loop Duration=0.27, UseLoopDelay=true)
   &FN_SpawnAnimTrail()
  ]
  particleSpawnScript [
   &FN_InitializeAnimTrailBoneIndecies(Second Bone or Socket Index=2) {module=FN_InitializeAnimTrailBoneIndecies001}
   &FN_InitializeAnimTrail(
    Color=&MultiplyLinearColor(
     Linear Color=<rgba 1 1 1 1>
     Scale Factor=&MakeVector4FromVectorAndFloat("Vector (RGB)"=<1 1 1>, "Float (Alpha)"=&Multiply_Float(A=$System.FadeSpawn, B=$System.FadeSpawnOut))
    )
    Lifetime=0.1
   ) {module=FN_InitializeAnimTrail001}
  ]
  particleUpdateScript [
   &FN_UpdateAnimTrail()
   &ScaleColor(Color Scale:LinearColor=$System.TODColor, Scale Mode=3) {module=ScaleColor001}
  ]
  rendererClasses [&NiagaraRibbonRendererProperties]
  renderers [
   {rendererClass=&NiagaraRibbonRendererProperties}
   {rendererClass=&NiagaraRibbonRendererProperties, rendererIndex=1}
  ]
 }
}
summary {
 systemName=NS_Pickaxe_DiscoButler_Trail
 emitters [
  {emitterName=TrailBlue, rendererClasses [&NiagaraRibbonRendererProperties]}
  {emitterName=TrailWater, rendererClasses [&NiagaraRibbonRendererProperties]}
  {emitterName=IceCubes, rendererClasses [&NiagaraMeshRendererProperties]}
 ]
 userVariables [
  User.RibbonIDOffset {type=NiagaraInt32}
  User.SwingSectionIndex {type=NiagaraInt32}
 ]
}
